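# Physics asset for the BakerStep drum kit cosmetic, defining jointed bodies for its hanging chain, ribbon and dagger chain.
# A five-link chain hangs from a hook locked to the hi-hat stand, with swing limits generally tightening toward the tip.
# A three-segment ribbon and short dagger chain attach to the handle.
# These joints mostly lock twist and limit swing, while the stand-to-handle constraint leaves rotation free.
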PhysicsAsset /SparksCosmetics/Instrument/DrumKitHighVert/BakerStep/Meshes/BakerStep_DrumKit_Physics
bodies {
 chain_01 [{shapeType=Capsule, length=2, radius=1.5, shapeName=chain_01_capsule, center=<1.6347 -0.2798 0.2385>, rotation=<rot 90 0 0>}]
 chain_02 [{shapeType=Capsule, length=1, radius=1.5, shapeName=chain_02_capsule, center=<2.0645 0.0205 -0.034>, rotation=<rot 90 0 0>}]
 chain_03 [{shapeType=Capsule, length=2, radius=1.5, shapeName=chain_03_capsule, center=<2.2298 0.1463 -0.1938>, rotation=<rot 90 0 0>}]
 chain_04 [{shapeType=Capsule, length=2, radius=1.5, shapeName=chain_04_capsule, center=<2.1337 0.1666 -0.3288>, rotation=<rot 90 0 0>}]
 chain_05 [{shapeType=Capsule, length=2, radius=1.5, shapeName=chain_05_capsule, center=<1.3816 0.1291 0.0613>, rotation=<rot 90 0 0>}]
 chain_hook [{shapeType=Sphere, radius=1, shapeName=chain_hook_sphere}]
 dagger_chain_01 [{shapeType=Capsule, length=0, radius=1.9775, shapeName=dagger_chain_01_capsule, center=<1.25 0 0>}]
 dagger_chain_root [{shapeType=Capsule, length=0, radius=1, shapeName=dagger_chain_root_capsule}]
 handle [{shapeType=Capsule, length=2, radius=1.5, rotation=<rot 90 0 0>}]
 hihat_stand [{shapeType=Capsule, length=19.9504, radius=2.3097, shapeName=hihat_stand_capsule, center=<50 -3.5672 1.5902>, rotation=<rot 81.7085 -90 -90>}]
 ribbon_01 [{shapeType=Capsule, length=2, radius=1.5, shapeName=ribbon_01_capsule, center=<1.0997 0.024 0.0788>, rotation=<rot 90 0 0>}]
 ribbon_02 [{shapeType=Capsule, length=2, radius=1.5, shapeName=ribbon_02_capsule, center=<-0.0599 -0.0569 0.0163>, rotation=<rot 90 0 0>}]
 ribbon_03 [{shapeType=Capsule, length=2, radius=1.5, shapeName=ribbon_03_capsule, center=<0.3056 0.0051 0.0019>, rotation=<rot 90 0 0>}]
}
constraints [
 {bone1Name=chain_05, bone2Name=chain_04, swing1LimitDegrees=15, swing2LimitDegrees=15, twistLimitDegrees=45, twistMotion=Locked}
 {bone1Name=chain_02, bone2Name=chain_01, swing1LimitDegrees=45, swing2LimitDegrees=20, twistLimitDegrees=45, twistMotion=Locked}
 {bone1Name=chain_03, bone2Name=chain_02, swing1LimitDegrees=20, swing2LimitDegrees=20, twistLimitDegrees=45, twistMotion=Locked}
 {bone1Name=chain_04, bone2Name=chain_03, swing1LimitDegrees=15, swing2LimitDegrees=15, twistLimitDegrees=45, twistMotion=Locked}
 {bone1Name=ribbon_03, bone2Name=ribbon_02, swing1LimitDegrees=5, swing2LimitDegrees=5, twistLimitDegrees=45, twistMotion=Locked}
 {bone1Name=ribbon_01, bone2Name=handle, swing1LimitDegrees=35, swing2LimitDegrees=35, twistLimitDegrees=45, twistMotion=Locked}
 {bone1Name=ribbon_02, bone2Name=ribbon_01, swing1LimitDegrees=35, swing2LimitDegrees=35, twistLimitDegrees=45, twistMotion=Locked}
 {bone1Name=chain_01, bone2Name=chain_hook, swing1LimitDegrees=40, swing2LimitDegrees=25, twistLimitDegrees=5}
 {
  bone1Name=chain_hook
  bone2Name=hihat_stand
  swing1LimitDegrees=35
  swing1Motion=Locked
  swing2LimitDegrees=35
  swing2Motion=Locked
  twistLimitDegrees=45
  twistMotion=Locked
 }
 {
  bone1Name=hihat_stand
  bone2Name=handle
  swing1LimitDegrees=45
  swing1Motion=Free
  swing2LimitDegrees=45
  swing2Motion=Free
  twistLimitDegrees=45
  twistMotion=Free
 }
 {
  bone1Name=dagger_chain_root
  bone2Name=handle
  swing1LimitDegrees=45
  swing1Motion=Locked
  swing2LimitDegrees=45
  swing2Motion=Locked
  twistLimitDegrees=45
  twistMotion=Locked
 }
 {
  bone1Name=dagger_chain_01
  bone2Name=dagger_chain_root
  swing1LimitDegrees=30
  swing2LimitDegrees=20
  twistLimitDegrees=45
  twistMotion=Locked
 }
]
